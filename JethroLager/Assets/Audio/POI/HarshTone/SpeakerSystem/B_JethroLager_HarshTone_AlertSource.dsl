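; B_JethroLager_HarshTone_AlertSource is an actor Blueprint that, by name, serves as an alert source for the HarshTone speaker system audio in JethroLager.
; BeginPlay and the construction script are empty.
; Two custom events, matching OnDestroyed and OnEndPlay signatures, each destroy this actor.
; Nothing shown binds or calls them; notes indicate they are disconnected hooks for removing the source when a wall or other object is destroyed.
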(event EventBeginPlay)

(event Custom|OnDestroyed_Event
  (params (DestroyedActor :type Actor%20Object%20Reference))
  ;[0.6574 0.1718 0.8594 0.5] Recconect and turn on variable if you want to destroy this actor when a wall or other is destroyed
  (Actor|DestroyActor))

(event Custom|OnEndPlay_Event
  (params (Actor :type Actor%20Object%20Reference) (EndPlayReason :type EEndPlayReason%20Enum :default "Destroyed"))
  ;[0.6574 0.1718 0.8594 0.5] Recconect and turn on variable if you want to destroy this actor when a wall or other is destroyed
  (Actor|DestroyActor))

(fn ConstructionScript)
