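; Music party controller for Jethro HarshTone speakers, repurposed as a zone alarm that sounds while enemy NPCs are alerted.
; Server-side box overlaps track players and bind alert, downed and destroyed events on unhired enemy-faction NPCs, caching alert levels in a map.
; Those events and a one-second poll evaluate Threatened/LostTarget, then toggle the party cosmetically via MultiX.
; StartMusicParty is gated on that alert state, and the actor self-destroys if placed near world origin.
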(event EventBeginPlay
  (if (Utilities|Operators|Equal(==) :A (Transformation|GetWorldLocation (Variables|Default|GetDefaultSceneRoot)) :ErrorTolerance 1e+03)
    (Actor|DestroyActor self)
    (else
      (Networking|CanExecuteCosmeticEvents
        (:False
          (bind _returnvalue (Utilities|Time|SetTimerbyEvent (delegate Custom|Loop_CheckAlarmStatus) 1.0 true))
          (Variables|Default|SetTmrCheckCurrentAlertStaus _returnvalue))))))

(event Custom|Loop_CheckAlarmStatus
  (bind _bisalerted (CallFunction|CheckCurrentAlertStatus))
  (CallFunction|MultiXToggleAlarms _bisalerted))

(event OnComponentBeginOverlap(Box)
  (params (OverlappedComponent :type Primitive%20Component%20Object%20Reference) (OtherActor :type Actor%20Object%20Reference) (OtherComp :type Primitive%20Component%20Object%20Reference) (OtherBodyIndex :type Integer) (bFromSweep :type Boolean) (SweepResult :type Hit%20Result%20Structure%20%28by%20ref%29))
  (bind _returnvalue (Fortnite|GetFortPlayerControllerfromActor OtherActor))
  ;[0 0 0 0.5] Server Only
  ;;[0.6574 0.1718 0.8594 0.5] If player character
  (Utilities|IsValid _returnvalue
    (:Is%20Valid
      (if (not (Utilities|Array|ContainsItem (Variables|Default|GetPlayersInZone) _returnvalue))
        (Utilities|Array|AddUnique (Variables|Default|GetPlayersInZone) _returnvalue))
      ;;[0.6574 0.1718 0.8594 0.5] Check on player enter to see if we should be playing alarms
      (bind _bisalerted (CallFunction|CheckCurrentAlertStatus))
      (CallFunction|MultiXToggleAlarms _bisalerted))
    (:Is%20Not%20Valid
      (bind (_bsucceded _asfort_athena_aibot_controller _asbp_player_pawn_non_participant) (CallFunction|ConfirmActorIsEnemyNPC OtherActor))
      (if _bsucceded
        (AlertLevel|AssignOnAlertLevelChangedEvent _asfort_athena_aibot_controller (EventDispatchers|CreateEvent :!event Custom|OnAlertLevelChangedEventDelegate_Event))
        (EventDispatchers|AssignOnDBNOStateChanged _asbp_player_pawn_non_participant (EventDispatchers|CreateEvent :!event Custom|OnDBNOStateChanged_Event))
        (Game|AssignOnDestroyed _asbp_player_pawn_non_participant (EventDispatchers|CreateEvent :!event Custom|OnDestroyed_Event))))))

(event Custom|NPCUnbindAlertLevelChanged
  (params (OldAlertLevel :type EAlertLevel%20Enum :default "Unaware") (NewAlertLevel :type EAlertLevel%20Enum :default "Unaware")))

(event Custom|NPCUnbindDBNO
  (params (FortPawn :type Fort%20Pawn%20Object%20Reference) (bInIsDBNO :type Boolean)))

(event Custom|NPCUnbindOnDestroyed
  (params (DestroyedActor :type Actor%20Object%20Reference)))

(event OnComponentEndOverlap(Box)
  (params (OverlappedComponent :type Primitive%20Component%20Object%20Reference) (OtherActor :type Actor%20Object%20Reference) (OtherComp :type Primitive%20Component%20Object%20Reference) (OtherBodyIndex :type Integer))
  (bind _returnvalue (Fortnite|GetFortPlayerControllerfromActor OtherActor))
  ;[0 0 0 0.5] Server Only - Clean up bindings when NPC leaves/dies or player leaves
  (Utilities|IsValid _returnvalue
    (:Is%20Valid
      (if (Utilities|Array|ContainsItem (Variables|Default|GetPlayersInZone) _returnvalue)
        (Utilities|Array|RemoveItem (Variables|Default|GetPlayersInZone) _returnvalue)
        (else
          _))
      (if (Utilities|Array|IsEmpty (Variables|Default|GetPlayersInZone))
        (CallFunction|MultiXToggleAlarms false)
        (else
          (bind _bisalerted (CallFunction|CheckCurrentAlertStatus))
          (CallFunction|MultiXToggleAlarms _bisalerted))))
    (:Is%20Not%20Valid
      (bind (_bsucceded _asfort_athena_aibot_controller _asbp_player_pawn_non_participant) (CallFunction|ConfirmActorIsEnemyNPC OtherActor))
      (if _bsucceded
        (Default|UnbindEventfromOnAlertLevelChanged _asbp_player_pawn_non_participant (delegate Custom|NPCUnbindAlertLevelChanged))
        (EventDispatchers|UnbindEventfromOnDBNOStateChanged _asbp_player_pawn_non_participant (delegate Custom|NPCUnbindDBNO))
        (Game|UnbindEventfromOnDestroyed _asbp_player_pawn_non_participant (delegate Custom|NPCUnbindOnDestroyed))
        (Utilities|Map|Remove (Variables|Default|GetNPCsAlertStatus) _asfort_athena_aibot_controller)
        ;;[0.8594 0.7556 0.1804 0.5] If the last NPC left for whatever reason then we should check
        (bind _bisalerted_2 (CallFunction|CheckCurrentAlertStatus))
        (CallFunction|MultiXToggleAlarms _bisalerted_2)))))

(event Custom|OnAlertLevelChangedEventDelegate_Event
  (params (BotController :type Fort%20Athena%20AIBot%20Controller%20Object%20Reference) (OldAlertLevel :type EAlertLevel%20Enum :default "Unaware") (NewAlertLevel :type EAlertLevel%20Enum :default "Unaware"))
  ;[0 0.6629 0.8594 0.5] Server side alert event
  (if (Utilities|Array|IsNotEmpty (Variables|Default|GetPlayersInZone))
    ;;[0.6574 0.1718 0.8594 0.5] Update each NPC's alert status when it changes so we have a constant updated list
    (Utilities|Map|Add (Variables|Default|GetNPCsAlertStatus) BotController NewAlertLevel)
    (bind _bisalerted (CallFunction|CheckCurrentAlertStatus))
    (CallFunction|MultiXToggleAlarms _bisalerted)))

(event Custom|MultiX_ToggleAlarms
  (params (bIsAlerted :type Boolean))
  ;[0.6574 0.1718 0.8594 0.5] Enable/Disable Alarms
  (Variables|Default|SetbAlertedStatusIsThreatened bIsAlerted)
  (Networking|CanExecuteCosmeticEvents
    (:True
      (if (Variables|Default|GetbAlertedStatusIsThreatened)
        (if (not (Variables|Default|GetAlarmsInProgress))
          (Class|MusicPartyControllerInterface|StartMusicParty)
          (Variables|Default|SetAlarmsInProgress true))
        (elif (Variables|Default|GetAlarmsInProgress)
          (Class|MusicPartyControllerInterface|StopMusicParty 0.0)
          (Variables|Default|SetAlarmsInProgress false))))))

(event EventTick
  (params (DeltaSeconds :type Float%20%28single-precision%29)))

(event Custom|OnDBNOStateChanged_Event
  (params (FortPawn :type Fort%20Pawn%20Object%20Reference) (bInIsDBNO :type Boolean))
  ;[0.6574 0.1718 0.8594 0.5] Update Alert Levels on DNBO
  (bind _bsucceeded (CallFunction|GetUpdateNPCAlertStatus FortPawn))
  (if _bsucceeded
    (bind _bisalerted (CallFunction|CheckCurrentAlertStatus))
    (CallFunction|MultiXToggleAlarms _bisalerted)))

(event Custom|OnDestroyed_Event
  (params (DestroyedActor :type Actor%20Object%20Reference))
  ;[0.6574 0.1718 0.8594 0.5] Clean up on NPC death
  (bind (_bsucceeded _asfort_athena_aibot_controller) (CallFunction|GetUpdateNPCAlertStatus DestroyedActor))
  (if _bsucceeded
    (Utilities|Map|Remove (Variables|Default|GetNPCsAlertStatus) _asfort_athena_aibot_controller)
    (else
      _))
  (bind _bisalerted (CallFunction|CheckCurrentAlertStatus))
  (CallFunction|MultiXToggleAlarms _bisalerted))

(event EventStartMusicParty
  ;[0.6574 0.1718 0.8594 0.5] Interface managment to Parent Class. Gating to prevent alarms going off if anything else tries to start the music other than NPCs
  (if (and (Variables|Default|GetbAlertedStatusIsThreatened) (not (Variables|Default|GetAlarmsInProgress)))
    (|CallParentFunction)))

(event EventStopMusicParty
  (params (FadeOutDuration :type Float%20%28double-precision%29))
  ;[0.6574 0.1718 0.8594 0.5] Interface managment to Parent Class. Gating to prevent alarms going off if anything else tries to start the music other than NPCs
  (|CallParentFunction 0.0))

(fn CheckCurrentAlertStatus
  (Networking|CanExecuteCosmeticEvents
    (:False
      (if (Utilities|Map|IsEmpty (Variables|Default|GetNPCsAlertStatus))
        (return false)
        (else
          (bind _keys (Utilities|Map|Keys (Variables|Default|GetNPCsAlertStatus)))
          (for _array_element _keys
            (Utilities|Map|Add (Variables|Default|GetNPCsAlertStatus) _array_element (Variables|/Script/FortniteAI.FortAthenaAIBotController|GetCurrentAlertLevel _array_element)))
          (bind _values (Utilities|Map|Values (Variables|Default|GetNPCsAlertStatus)))
          (bind _array_element_2 (Utilities|Array|ForEachLoopwithBreak _values)
            (:LoopBody
              (if (or (Utilities|Enum|Equal(Enum) _array_element_2 "Threatened") (Utilities|Enum|Equal(Enum) _array_element_2 "LostTarget"))
                (LocalVariables|SetAlerted true))
              (LocalVariables|SetAlerted false))
            (:Completed
              _))
          (return (LocalVariables|GetAlerted)))))))

(fn CheckIfNPCIsEnemyFaction
  (params (Actor :type Actor%20Object%20Reference))
  (return (GameplayTags|HasAnyTags (Fortnite|Factions|GetFactionTagContainerforActor Actor) (GameplayTags|MakeLiteralGameplayTagContainer "(GameplayTags=((TagName=\"Athena.Faction.Jethro.LoneFeather.RusticMouse\"),(TagName=\"Athena.Faction.Jethro.MistyFang\"),(TagName=\"Athena.Faction.Jethro.LavaTrunk\"),(TagName=\"Athena.Faction.Jethro.RusticMouse\"),(TagName=\"Athena.Faction.Jethro.WingBath\")))"))))

(fn CheckIfNPCIsHired
  (params (Actor :type Fort%20Player%20Pawn%20Object%20Reference))
  (bind _convertcomponent (Variables|/Script/FortniteGame.FortPlayerPawn|GetConvertComponent Actor)
    (:then
      (Utilities|IsValid (Convert|GetConverterPawn _convertcomponent)
        (:Is%20Valid
          _)
        (:Is%20Not%20Valid
          (return true))))
    (:else
      _))
  (return false))

(fn ConfirmActorIsEnemyNPC
  (params (NPC :type Object%20Reference))
  (bind (_bsucceeded _asfort_athena_aibot_controller _asbp_player_pawn_non_participant) (CallFunction|GetUpdateNPCAlertStatus NPC))
  (if _bsucceeded
    (bind _bisenemy (CallFunction|CheckifNPCIsEnemyFaction _asbp_player_pawn_non_participant))
    (if _bisenemy
      (bind _bisnothireable (CallFunction|CheckifNPCIsHired _asbp_player_pawn_non_participant))
      (if _bisnothireable
        (return true _asfort_athena_aibot_controller _asbp_player_pawn_non_participant)
        (else
          _))
      (else
        _))
    (return false 0 0)
    (else
      _)))

(fn GetNPCController
  (params (Object :type Object%20Reference))
  (bind _asbp_player_pawn_non_participant (Utilities|Casting|CastToBP_PlayerPawn_NonParticipant Object)
    (:then
      (Utilities|IsValid (Pawn|GetController _asbp_player_pawn_non_participant)
        (:Is%20Valid
          (bind _asfort_athena_aibot_controller (Utilities|Casting|CastToFortAthenaAIBotController (Pawn|GetController _asbp_player_pawn_non_participant))
            (:then
              (return true _asfort_athena_aibot_controller _asbp_player_pawn_non_participant))
            (:CastFailed
              _)))
        (:Is%20Not%20Valid
          _))
      (return false 0 0))
    (:CastFailed
      _)))

(fn GetUpdateNPCAlertStatus
  (params (NPC :type Object%20Reference))
  (bind (_bsucceded _asfort_athena_aibot_controller _asbp_player_pawn_non_participant) (CallFunction|GetNPCController NPC))
  (if _bsucceded
    (Utilities|Map|Add (Variables|Default|GetNPCsAlertStatus) _asfort_athena_aibot_controller (Variables|/Script/FortniteAI.FortAthenaAIBotController|GetCurrentAlertLevel _asfort_athena_aibot_controller))
    (return true _asfort_athena_aibot_controller _asbp_player_pawn_non_participant)
    (else
      (return false 0 0))))

(fn ConstructionScript)
