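# Mono MetaSound source for the looping alarm of the JethroLager HarshTone speaker system, reusing the BinGrass boss fight alarm wave.
# On play, it starts the wave player and triggers an ADSR envelope that swells in slowly, then decays to a decibel-defined sustain level.
# The envelope, boosted by a multiplier, scales the audio.
# Base submix and submix sends are off, so output is routed to the HarshTone audio bus.
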MetaSoundSource /JethroLager/Assets/Audio/POI/HarshTone/SpeakerSystem/MSS_JethroLager_HarshTone_Alarm_Lp
inputs [UE.Source.OnPlay {type=Trigger}]
interfaces [UE.OutputFormat.Mono:1.0, UE.Source:1.1]
outputs ["UE.OutputFormat.Mono.Audio:0" {type=Audio}]
pages [Default {
  input_defaults {UE.Source.OnPlay=null}
  nodes [
   DecibelstoLinearGain_0 = UE.Decibels to Linear Gain:Float() {also {defaults {Decibels=-6}}}
   ADSREnvelopeADSREnvelope_0 = ADSR Envelope.ADSR Envelope:Float(Trigger Attack=^UE.Source.OnPlay, Sustain Level=@DecibelstoLinearGain_0.Linear Gain) {also {defaults {Attack Time=2, Decay Time=5, Hard Reset="<MetaSoundBuilderResult.SUCCEEDED: 0>"}}}
   WavePlayer_0 = UE.Wave Player:Mono(Play=^UE.Source.OnPlay) {also {
     defaults {Loop="<MetaSoundBuilderResult.SUCCEEDED: 0>", Wave Asset=/JethroLager/NPCs/Bosses/BinGrass/Audio/Waves/BinGrass_BossFight_Alarm_Loop}
     version [1,1]
    }}
   Multiply_0 = UE.Multiply:Audio by Float(
    AdditionalOperands=UE.Multiply:Float(PrimaryOperand=@ADSREnvelopeADSREnvelope_0.Out Envelope) {also {defaults {AdditionalOperands=3}}}.Out
    PrimaryOperand=@WavePlayer_0.Out Mono
   )
  ]
  links [{dst="@UE.OutputFormat.Mono.Audio:0", src=Multiply_0.Out}]
 }]
properties {
 assetImportData=&self:AssetImportData
 bEnableBaseSubmix=false
 bEnableSubmixSends=false
 bLooping=true
 duration=10000
 sampleRate=48000
 busSends [{audioBus=&/AB_JethroLager_HarshTone, sendLevel=1}]
}
